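; Fortnite Creative Creature Manager device that overrides stats, scoring and knockback for a chosen creature type and broadcasts an event when a matching creature is eliminated.
; SetPresetValues turns user options into selected overrides, using invalid sentinels for unset ones; minigame and custom events toggle enabled state.
; Verse accessors read, clamp-set or reset values on the CreatureInfoComponent.
; Legacy presets migrate by comparison with class defaults, and the preview mesh follows creature type.
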(event EventOnWorldReady
  ;[0 1 0.08672] Visible In Game Component needs initialized even when working in basic mode as BP components can't initialize in editor,
  ; and BuildingProp based devices need to initialize visibleingame OnWorldReady
  (|CallParentFunction)
  (Class|CreativeVisibleinGameComponent|InitializeComponent (Variables|Default|GetVisibleInGameComponent)))

(event EditorPreview|EventSetupForEditorPreview
  ;[0 1 0.08672] Visible In Game Component needs initialized even when working in basic mode as BP components can't initialize in editor,
  ; and BuildingProp based devices need to initialize visibleingame OnWorldReady
  (|CallParentFunction)
  (Class|CreativeVisibleinGameComponent|InitializeComponent (Variables|Default|GetVisibleInGameComponent)))

(event EventSetupConfigurables
  (CallFunction|SetPresetValues))

(event OnAnyPropertyChanged(ToyOptionsComponent)
  (Variables|BuildingProp_CreatureManager|SetSelectedAIType (Utilities|Array|Get(acopy) (Variables|CustomizeOptions|GetAITypeOptions) (Default|GetCreatureTypeasInt)))
  (CallFunction|LoadSkeletalMeshDisplayOption)
  (Visibility|SetVisibilityComponents (Variables|Default|GetVisibleInGameComponent) (Utilities|Array|MakeArray (Variables|Default|GetAITypeMesh) (Variables|Default|GetBaseMesh)) true))

(event OnPropertiesLoaded(ToyOptionsComponent)
  (Variables|BuildingProp_CreatureManager|SetSelectedAIType (Utilities|Array|Get(acopy) (Variables|CustomizeOptions|GetAITypeOptions) (Default|GetCreatureTypeasInt)))
  (CallFunction|LoadSkeletalMeshDisplayOption)
  (Visibility|SetVisibilityComponents (Variables|Default|GetVisibleInGameComponent) (Utilities|Array|MakeArray (Variables|Default|GetAITypeMesh) (Variables|Default|GetBaseMesh)) true))

(event EventSkeletalMeshLoadComplete
  (params (Material :type Material%20Interface%20Object%20Reference) (Mesh :type Skeletal%20Mesh%20Object%20Reference))
  (CallFunction|UpdateVisibleSkeletalMesh Material Mesh))

(event OnVisibilityChanged(VisibleInGameComponent)
  (params (Visible :type Boolean))
  (CallFunction|AdjustCollisionOfStaticMesh (not Visible) (Variables|Debug|GetStaticMeshComponent)))

(event Custom|EnableEvent
  (params (Instigator :type Controller%20Object%20Reference))
  (CallFunction|NotifyDeviceEnableChanged true))

(event MinigameStartedDelegate(FortMinigameProgress)
  (Networking|SwitchHasAuthority
    (:Authority
      (CallFunction|NotifyDeviceEnableChanged (Variables|UserOptions|GetEnabled%20on%20Game%20Start)))))

(event Custom|DisableEvent
  (params (Instigator :type Controller%20Object%20Reference))
  (CallFunction|NotifyDeviceEnableChanged false))

(event MinigameEndedDelegate(FortMinigameProgress)
  (Networking|SwitchHasAuthority
    (:Authority
      (CallFunction|NotifyDeviceEnableChanged false))))

(event EventOnManagedCreatureDied
  (params (InstigatingController :type Controller%20Object%20Reference) (DamagedActor :type Actor%20Object%20Reference))
  (GameplayEvent|BroadcastEvent (Variables|UserOptions|Events|GetOn%20a%20Matching%20Creature%20Type%20is%20Eliminated) 0 (Utilities|Struct|MakeGameplayEventManagedCreatureEliminated InstigatingController)))

(fn Get_AllowWeaponKnockback_Verse
  (return (Class|FortCreativeCreatureManagerInfoComponent|GetWeaponKnockbackImmunity (Variables|BuildingProp_CreatureManager|GetCreatureInfoComponent))))

(fn Get_DamageToEnvironment_Verse
  (return (Class|FortCreativeCreatureManagerInfoComponent|GetEnvironmentalDamageOverride (Variables|BuildingProp_CreatureManager|GetCreatureInfoComponent) :!class /Script/FortniteAI.FortCreativeCreatureManagerInfoComponent)))

(fn Get_DamageToPlayer_Verse
  (return (Class|FortCreativeCreatureManagerInfoComponent|GetOverrideDamage (Variables|BuildingProp_CreatureManager|GetCreatureInfoComponent))))

(fn Get_MaxHealth_Verse
  (return (Class|FortCreativeCreatureManagerInfoComponent|GetMaxHealth (Variables|BuildingProp_CreatureManager|GetCreatureInfoComponent) :!class /Script/FortniteAI.FortCreativeCreatureManagerInfoComponent)))

(fn Get_MovementSpeed_Verse
  (return (Class|FortCreativeCreatureManagerInfoComponent|GetMovementSpeedMultiplier (Variables|BuildingProp_CreatureManager|GetCreatureInfoComponent) :!class /Script/FortniteAI.FortCreativeCreatureManagerInfoComponent)))

(fn Get_Score_Verse
  (return (Class|FortCreativeCreatureManagerInfoComponent|GetScorePoints (Variables|BuildingProp_CreatureManager|GetCreatureInfoComponent) :!class /Script/FortniteAI.FortCreativeCreatureManagerInfoComponent)))

(fn GetCreatureTypeAsInt
  (return (- (Math|Conversions|ToInteger(Byte) (Variables|UserOptions|Basic|GetCreature%20Type)) 1)))

(fn MigrateLegacyUserOptions
  (bind _aiscorepresetvalue (Class|GetClassDefaults "/CRD_Creatures/Device_CreatureManager_V2.Device_CreatureManager_V2_C"))
  (bind _aitypepresetindex (Class|GetClassDefaults "/CRD_Creatures/Device_CreatureManager_V2.Device_CreatureManager_V2_C"))
  (bind _aihealthpresetvalue (Class|GetClassDefaults "/CRD_Creatures/Device_CreatureManager_V2.Device_CreatureManager_V2_C"))
  (bind _aimovementspeedmultiplierpresetvalue (Class|GetClassDefaults "/CRD_Creatures/Device_CreatureManager_V2.Device_CreatureManager_V2_C"))
  (bind _aienvironmentaldamagepresetvalue (Class|GetClassDefaults "/CRD_Creatures/Device_CreatureManager_V2.Device_CreatureManager_V2_C"))
  (bind _aidamagepresetvalue (Class|GetClassDefaults "/CRD_Creatures/Device_CreatureManager_V2.Device_CreatureManager_V2_C"))
  (bind _aitypepresetindex_2 (Variables|Legacy|GetAITypePresetIndex))
  (bind _aihealthpresetvalue_2 (Variables|Legacy|GetAIHealthPresetValue))
  (bind _aienvironmentaldamagepresetvalue_2 (Variables|Legacy|GetAIEnvironmentalDamagePresetValue))
  (bind _aimovementspeedmultiplierpresetvalue_2 (Variables|Legacy|GetAIMovementSpeedMultiplierPresetValue))
  (bind _aidamagepresetvalue_2 (Variables|Legacy|GetAIDamagePresetValue))
  (bind _aiscorepresetvalue_2 (Variables|Legacy|GetAIScorePresetValue))
  (|CallParentFunction)
  (Utilities|FlowControl|Sequence
    (:then_0
      (Utilities|FlowControl|Sequence
        (:then_0
          (if (!= _aitypepresetindex _aitypepresetindex_2)
            (Variables|UserOptions|Basic|SetCreature%20Type (Utilities|Enum|BytetoEnumECreatureType (Math|Conversions|ToByte(Integer) (+ _aitypepresetindex_2 1))))
            (Variables|Legacy|SetAITypePresetIndex _aitypepresetindex)
            (LocalVariables|SetOptions%20Migrated true)))
        (:then_1
          (if (!= _aihealthpresetvalue _aihealthpresetvalue_2)
            (Variables|UserOptions|SetHealth_Override true)
            (Variables|UserOptions|Basic|SetHealth _aihealthpresetvalue_2)
            (Variables|Legacy|SetAIHealthPresetValue _aihealthpresetvalue)
            (LocalVariables|SetOptions%20Migrated true)))
        (:then_2
          (if (!= _aimovementspeedmultiplierpresetvalue _aimovementspeedmultiplierpresetvalue_2)
            (Variables|UserOptions|SetMovementSpeed_Override true)
            (Variables|UserOptions|Basic|SetMovementSpeed _aimovementspeedmultiplierpresetvalue_2)
            (Variables|Legacy|SetAIMovementSpeedMultiplierPresetValue _aimovementspeedmultiplierpresetvalue)
            (LocalVariables|SetOptions%20Migrated true)))
        (:then_3
          (if (!= _aidamagepresetvalue _aidamagepresetvalue_2)
            (Variables|UserOptions|SetDamageToPlayer_Override true)
            (Variables|UserOptions|Basic|SetDamageToPlayer _aidamagepresetvalue_2)
            (Variables|Legacy|SetAIDamagePresetValue _aidamagepresetvalue)
            (LocalVariables|SetOptions%20Migrated true)))
        (:then_4
          (if (!= _aienvironmentaldamagepresetvalue _aienvironmentaldamagepresetvalue_2)
            (Variables|UserOptions|SetDamageToEnvironment_Override true)
            (Variables|UserOptions|SetDamageToEnvironment _aienvironmentaldamagepresetvalue_2)
            (Variables|Legacy|SetAIEnvironmentalDamagePresetValue _aienvironmentaldamagepresetvalue)
            (LocalVariables|SetOptions%20Migrated true)))
        (:then_5
          (if (!= _aiscorepresetvalue _aiscorepresetvalue_2)
            (Variables|UserOptions|SetScore_Override true)
            (Variables|UserOptions|SetScore _aiscorepresetvalue_2)
            (Variables|Legacy|SetAIScorePresetValue _aiscorepresetvalue)
            (LocalVariables|SetOptions%20Migrated true)))))
    (:then_1
      (if (LocalVariables|GetOptions%20Migrated)
        (OptionData|RefreshPropertyOverrides (Variables|Default|GetToyOptionsComponent))))))

(fn Reset_DamageToEnvironment_Verse
  (Class|FortCreativeCreatureManagerInfoComponent|SetEnvironmentalDamageOverride (Variables|BuildingProp_CreatureManager|GetCreatureInfoComponent) -1.0 true))

(fn Reset_DamageToPlayer_Verse
  (Class|FortCreativeCreatureManagerInfoComponent|SetOverrideDamage (Variables|BuildingProp_CreatureManager|GetCreatureInfoComponent) -1.0 true))

(fn Reset_MaxHealth_Verse
  (Class|FortCreativeCreatureManagerInfoComponent|SetMaxHealth (Variables|BuildingProp_CreatureManager|GetCreatureInfoComponent) -1 true))

(fn Reset_MovementSpeed_Verse
  (Class|FortCreativeCreatureManagerInfoComponent|SetMovementSpeedMultiplier (Variables|BuildingProp_CreatureManager|GetCreatureInfoComponent) -1.0 true))

(fn Reset_Score_Verse
  (Class|FortCreativeCreatureManagerInfoComponent|SetScorePoints (Variables|BuildingProp_CreatureManager|GetCreatureInfoComponent) -1 true))

(fn Set_AllowWeaponKnockback_Verse
  (params (NewAllowWeaponKnockback :type Boolean))
  (Class|FortCreativeCreatureManagerInfoComponent|SetWeaponKnockbackImmunity (Variables|BuildingProp_CreatureManager|GetCreatureInfoComponent) NewAllowWeaponKnockback true))

(fn Set_DamageToEnvironment_Verse
  (params (NewDamageToEnvironment :type Float%20%28double-precision%29))
  (Class|FortCreativeCreatureManagerInfoComponent|SetEnvironmentalDamageOverride (Variables|BuildingProp_CreatureManager|GetCreatureInfoComponent) (Math|Float|Clamp(Float) NewDamageToEnvironment 0.0 5e+02) true))

(fn Set_DamageToPlayer_Verse
  (params (NewDamageToPlayer :type Float%20%28double-precision%29))
  (Class|FortCreativeCreatureManagerInfoComponent|SetOverrideDamage (Variables|BuildingProp_CreatureManager|GetCreatureInfoComponent) (Math|Float|Clamp(Float) NewDamageToPlayer 0.0 5e+02) true))

(fn Set_MaxHealth_Verse
  (params (NewHealth :type Integer))
  (Class|FortCreativeCreatureManagerInfoComponent|SetMaxHealth (Variables|BuildingProp_CreatureManager|GetCreatureInfoComponent) (Math|Integer|Clamp(Integer) NewHealth 1 10000) true))

(fn Set_MovementSpeed_Verse
  (params (NewMovementSpeed :type Float%20%28double-precision%29))
  (Class|FortCreativeCreatureManagerInfoComponent|SetMovementSpeedMultiplier (Variables|BuildingProp_CreatureManager|GetCreatureInfoComponent) (Math|Float|Clamp(Float) NewMovementSpeed 0.0 2.0) true))

(fn Set_Score_Verse
  (params (NewScore :type Integer))
  (Class|FortCreativeCreatureManagerInfoComponent|SetScorePoints (Variables|BuildingProp_CreatureManager|GetCreatureInfoComponent) (Math|Integer|Clamp(Integer) NewScore 0 100) true))

(fn SetPresetValues
  (bind _creatureinfocomponent (Variables|BuildingProp_CreatureManager|GetCreatureInfoComponent))
  (bind _returnvalue (Class|FortCreativeCreatureManagerInfoComponent|GetInvalidInt32 _creatureinfocomponent))
  (bind _creatureinfocomponent_2 (Variables|BuildingProp_CreatureManager|GetCreatureInfoComponent))
  (bind _returnvalue_2 (Class|FortCreativeCreatureManagerInfoComponent|GetInvalidFloat _creatureinfocomponent_2))
  (Variables|BuildingProp_CreatureManager|SetSelectedAIType (Utilities|Array|Get(acopy) (Variables|CustomizeOptions|GetAITypeOptions) (Default|GetCreatureTypeasInt)))
  (Variables|BuildingProp_CreatureManager|SetSelectedOverrideHealth (Math|Integer|SelectInt (Variables|UserOptions|Basic|GetHealth) _returnvalue (Variables|UserOptions|GetHealth_Override)))
  (Variables|BuildingProp_CreatureManager|SetSelectedOverrideScoreValue (Math|Integer|SelectInt (Variables|UserOptions|GetScore) _returnvalue (Variables|UserOptions|GetScore_Override)))
  (Variables|BuildingProp_CreatureManager|SetSelectedOverrideAggroDistance (* (Variables|Default|GetAIAggroRangePresetValue) 512.0))
  (Variables|BuildingProp_CreatureManager|SetSelectedOverrideDamage (Math|Float|SelectFloat (Variables|UserOptions|Basic|GetDamageToPlayer) _returnvalue_2 (Variables|UserOptions|GetDamageToPlayer_Override)))
  (Variables|BuildingProp_CreatureManager|SetSelectedEnvironmentalDamageOverride (Math|Float|SelectFloat (Variables|UserOptions|GetDamageToEnvironment) _returnvalue_2 (Variables|UserOptions|GetDamageToEnvironment_Override)))
  (Variables|BuildingProp_CreatureManager|SetSelectedMovementSpeedMultiplier (Math|Float|SelectFloat (Variables|UserOptions|Basic|GetMovementSpeed) _returnvalue_2 (Variables|UserOptions|GetMovementSpeed_Override)))
  (Variables|BuildingProp_CreatureManager|SetSelectedOverrideScoreDistribution (Variables|UserOptions|GetScoreDistribution))
  (Variables|BuildingProp_CreatureManager|SetbImmuneToWeaponKnockback (not (Variables|UserOptions|GetAllow%20Weapon%20Knockback)))
  (Variables|BuildingProp_CreatureManager|SetSelectedAffectedTargetsType (Variables|UserOptions|GetAffected%20Creatures)))

(fn UpdateVisibleSkeletalMesh
  (params (Material :type Material%20Interface%20Object%20Reference) (Mesh :type Skeletal%20Mesh%20Object%20Reference))
  (Components|SkinnedMesh|SetSkinnedAssetandUpdate (Variables|Default|GetAITypeMesh) (Variables|GetMesh) true)
  (Rendering|Material|SetMaterial (Variables|Default|GetAITypeMesh) 0 (Variables|GetMaterial)))

(fn ConstructionScript
  (bind _asint (Default|GetCreatureTypeasInt))
  ;[0.9331 0.3749 1] Update the skeletal mesh based on the current creature type. Use the creature type enumerator -1 to get a valid index because the enumerator first index is 1 whereas the AITypeOptions first index is 0.
  (if (and (Editor|InEditor) (Utilities|Array|IsValidIndex (Variables|CustomizeOptions|GetAITypeOptions) _asint))
    (Variables|BuildingProp_CreatureManager|SetSelectedAIType (Utilities|Array|Get(acopy) (Variables|CustomizeOptions|GetAITypeOptions) _asint))
    (CallFunction|LoadSkeletalMeshDisplayOption)))
