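; Creative Creature Placer device that spawns a selected creature type through a creative rift.
; User options drive spawn phase, despawn and activation settings via SetupConfigurables, while an async-loaded idle animation previews the creature.
; Spawn and kill notifications apply gameplay tags, fade ambient audio and broadcast events; Verse can queue transform spawns tracked by request ID.
; It also migrates the legacy creature index and registers a nav invoker once playlist data loads.
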(event EventBeginPlay
  (Utilities|FlowControl|Sequence
    (:then_0
      (Creative|Networking|Authority/Cosmetic
        (:Authority
          (CallFunction|TryRegisterNavInvoker))))
    (:then_1
      (Creative|Networking|Authority/Cosmetic
        (:Cosmetic
          (Variables|/Script/Engine.SkeletalMeshComponent|SetbNoSkeletonUpdate (Building|IsPreviewBuildingActor) (Variables|Debug|GetCreatureSkeletalMeshComponent)) ; Turning off skeleton updates while it's a preview fixes placement bounds
          (EventDispatchers|AssignOnCreativeAssetsLoaded (Variables|BuildingProp_CreaturePlacer|GetAssetLoadingTrackingComponent) (EventDispatchers|CreateEvent :!event Custom|OnCreatureVisualsLoadCompleteClient)))))
    (:then_2
      (CallFunction|SetSpawnLocation (Variables|Default|GetSpawnPoint)))))

(event EventOnWorldReady
  (|CallParentFunction)
  (Visibility|SetVisibilityComponents (Variables|Default|GetMeshesVisibleInGameComponent) (Utilities|Array|MakeArray (Variables|Debug|GetStaticMeshComponent) (Variables|Debug|GetCreatureSkeletalMeshComponent)) true)
  (Visibility|SetVisibilityComponents (Variables|Default|GetVisibleInGameComponent) (Utilities|Array|MakeArray (Variables|Default|GetCreatureHighlight) (Variables|Default|GetHoloFloor) (Variables|Default|GetLightRaysA) (Variables|Default|GetLightRaysB) (Variables|Default|GetSpotLight) (Variables|Default|GetP_CreaturePlacer_Smoke)) true)
  (CallFunction|InitializeDevice))

(event EditorPreview|EventSetupForEditorPreview
  (|CallParentFunction)
  (Visibility|SetVisibilityComponents (Variables|Default|GetMeshesVisibleInGameComponent) (Utilities|Array|MakeArray (Variables|Debug|GetStaticMeshComponent) (Variables|Debug|GetCreatureSkeletalMeshComponent)) true)
  (Visibility|SetVisibilityComponents (Variables|Default|GetVisibleInGameComponent) (Utilities|Array|MakeArray (Variables|Default|GetCreatureHighlight) (Variables|Default|GetHoloFloor) (Variables|Default|GetLightRaysA) (Variables|Default|GetLightRaysB) (Variables|Default|GetSpotLight) (Variables|Default|GetP_CreaturePlacer_Smoke)) true)
  (CallFunction|InitializeDevice))

(event EventMinigameStateChanged
  (params (Minigame :type Fort%20Minigame%20Object%20Reference) (NewMinigameState :type EFortMinigameState%20Enum :default "PreGame"))
  (bind _returnvalue (CallFunction|GetCreativeRift))
  (if (Utilities|Enum|Equal(Enum) NewMinigameState "PreGame")
    (CallFunction|SetupConfigurables)
    (else
      _))
  (|CallParentFunction Minigame NewMinigameState)
  (if (and (Utilities|Enum|Equal(Enum) NewMinigameState "Setup") (Utilities|IsValid _returnvalue :!class /Script/Engine.KismetSystemLibrary))
    (Class|BuildingRift|EnableSpawning _returnvalue (not (Variables|UserOptions|GetDelay%20Spawn%20with%20Encounter%20Start)))))

(event OnAnyPropertyChanged(ToyOptionsComponent)
  (CallFunction|InitializeDevice)
  (Class|CreativePhotoModeVisibilityComponent|SetComponentstoConditionallyHide (Variables|Default|GetPhotoModeVisibilityComponent) (CallFunction|PhotoModeGetAllConditionallyHiddenComponents)))

(event OnPropertiesLoaded(ToyOptionsComponent)
  (CallFunction|InitializeDevice)
  (Class|CreativePhotoModeVisibilityComponent|SetComponentstoConditionallyHide (Variables|Default|GetPhotoModeVisibilityComponent) (CallFunction|PhotoModeGetAllConditionallyHiddenComponents)))

(event OnVisibilityChanged(MeshesVisibleInGameComponent)
  (params (Visible :type Boolean))
  (bind _creatureskeletalmeshcomponent (Variables|Debug|GetCreatureSkeletalMeshComponent))
  (bind _staticmeshcomponent (Variables|Debug|GetStaticMeshComponent))
  (if Visible
    (CallFunction|AdjustCollision false _creatureskeletalmeshcomponent)
    (CallFunction|AdjustCollision false _staticmeshcomponent)
    (else
      (CallFunction|AdjustCollision true _creatureskeletalmeshcomponent)
      (CallFunction|AdjustCollision true _staticmeshcomponent))))

(event Custom|OnCreatureVisualsLoadCompleteClient
  (bind _creature_index (CallFunction|GetCreatureTypeIndex))
  (bind _enemyidleanimations (Variables|Default|GetEnemyIdleAnimations))
  (bind _output (Utilities|Array|Get(acopy) _enemyidleanimations _creature_index))
  (bind _creatureskeletalmeshcomponent (Variables|Debug|GetCreatureSkeletalMeshComponent))
  ;[1 1 1] Client Visuals
  (if (Utilities|IsValidSoftObjectReference _output)
    (bind _loaded_object (Creative|Loading|AsyncLoadEditorSafe _output)
      (:Loaded
        (Utilities|FlowControl|Sequence
          (:then_0
            (bind _asanim_sequence (Utilities|Casting|CastToAnimSequence _loaded_object))
            (if (not (CallFunction|IsPlayingAnimation _asanim_sequence))
              (Components|Animation|PlayAnimation _creatureskeletalmeshcomponent _asanim_sequence (Variables|Default|GetLoopingPreviewAnimation))
              (Variables|/Script/Engine.SkeletalMeshComponent|SetAnimationData (Utilities|Struct|MakeSingleAnimationPlayData _asanim_sequence (Variables|Default|GetLoopingPreviewAnimation) (Components|Animation|IsPlaying (Variables|Debug|GetCreatureSkeletalMeshComponent)) (Components|Animation|GetPosition (Variables|Debug|GetCreatureSkeletalMeshComponent)) (Components|Animation|GetPlayRate (Variables|Debug|GetCreatureSkeletalMeshComponent))) _creatureskeletalmeshcomponent)
              ;[0.5117 1 0.5418] Play animation in editor
              (if (Editor|InEditor)
                (Components|Animation|SetAnimationMode _creatureskeletalmeshcomponent "AnimationSingleNode" true)
                (Components|Animation|SetAnimation _creatureskeletalmeshcomponent _asanim_sequence)
                (Components|SkeletalMesh|SetUpdateAnimationinEditor _creatureskeletalmeshcomponent true))))
          (:then_1
            (Components|Activation|Activate (Variables|Default|GetP_CreaturePlacer_Transition) true)
            (if (not (Editor|InEditor))
              (Audio|PlaySoundatLocation "/Game/Sounds/Creative/Toys/CreaturePlacer/CreaturePlacer_CreatureAppears_Cue.CreaturePlacer_CreatureAppears_Cue" (Transformation|GetActorLocation)))))))))

(event EventBP_NotifyAIPawnSpawned
  (params (SpawnedAI :type Fort%20AIPawn%20Object%20Reference) (SpawnSlotLocation :type Vector%20%28by%20ref%29 :default "0, 0, 0"))
  (bind _returnvalue (CallFunction|GetCreativeRift))
  (bind _aipawnabilitysystemcomponent (Variables|/Script/FortniteGame.FortAIPawn|GetAIPawnAbilitySystemComponent SpawnedAI))
  (Utilities|IsValid SpawnedAI
    (:Is%20Valid
      (if (not (Variables|UserOptions|Basic|GetRestore%20Player%20Shield%20On%20Elimination))
        (Ability|GameplayEffect|AddLooseGameplayTags SpawnedAI (GameplayTags|MakeLiteralGameplayTagContainer "(GameplayTags=((TagName=\"Pawn.Athena.NPC.Deimos.DisableShieldOnElimination\")))") true) ; FORT-684077 | Add Tag to disallow shield rewards
      )
      (bind _returnvalue_2 (Ability|GameplayEffect|AddGrantedTag (GameplayEffects|MakeOutgoingSpec :self _aipawnabilitysystemcomponent :GameplayEffectClass "/Game/Creative/Devices/CreaturePlacer/GameplayEffects/GE_EliminationManagerAddTag.GE_EliminationManagerAddTag_C" :Level 0.0) (Variables|Default|GetSpawnAppliedGameplayTag)))
      (GameplayEffects|ApplyGameplayEffectSpecToSelf _aipawnabilitysystemcomponent _returnvalue_2)
      (CallFunction|FadeAmbientLoop false) ; Fade out because the device won't spawn anything else until it's told to turn back on
      (GameplayEvent|BroadcastEvent (Variables|UserOptions|Events|GetOn%20Spawned) 0 (Utilities|Struct|MakeGameplayEventCreatureSpawned (Pawn|GetController SpawnedAI)))
      (if (Utilities|IsValid _returnvalue :!class /Script/Engine.KismetSystemLibrary)
        (Class|BuildingRift|EnableSpawning _returnvalue (not (Variables|UserOptions|GetDelay%20Spawn%20with%20Encounter%20Start)))))))

(event EventBP_NotifyAIPawnKilled
  (params (KilledAI :type Fort%20AIPawn%20Object%20Reference) (KillingInstigator :type Controller%20Object%20Reference))
  (GameplayEvent|BroadcastEvent (Variables|UserOptions|Events|GetOn%20Eliminated) 0 (Utilities|Struct|MakeGameplayEventCreatureEliminated KillingInstigator)))

(event Custom|EnableEvent
  (params (Instigator :type Controller%20Object%20Reference))
  (CallFunction|OnTurnonReceived Instigator)
  (CallFunction|FadeAmbientLoop true))

(event Custom|DisableEvent
  (params (Instigator :type Controller%20Object%20Reference))
  (CallFunction|OnTurnOffReceived Instigator)
  (CallFunction|FadeAmbientLoop false))

(event OnAddedtoMinigame(MinigameProgress)
  (params (Minigame :type Fort%20Minigame%20Object%20Reference))
  (Utilities|IsValid Minigame
    (:Is%20Valid
      (Class|CreativePhotoModeVisibilityComponent|SetComponentstoAlwaysHide (Variables|Default|GetPhotoModeVisibilityComponent) (CallFunction|PhotoModeGetAllHiddenDuringGameComponents)))))

(event Custom|TryRegisterNavInvoker
  (bind _self self)
  (bind _returnvalue (Fortnite|Athena|GameState|GetGameStateAthenafromContext _self))
  ;[1 1 1] Add NavInvoker around device when required
  (if (Variables|/Script/FortniteGame.FortGameStateAthena|GetbPlaylistDataIsLoaded _returnvalue)
    (Class|BPCreativeCommonAILibrary|TryAddNavInvokeronDevice self "/CRD_Creatures/Components/CreaturePlacer_NavInvokerCompClass.CreaturePlacer_NavInvokerCompClass_C") ; Will add Nav Invoker around this device if Nav Mesh on Demand is enabled
    (else
      (Playlist|BindEventtoOnCurrentPlaylistLoaded _returnvalue (delegate Custom|OnPlaylistLoaded)))))

(event Custom|OnPlaylistLoaded
  (params (PlaylistName :type Name) (PlaylistContextTags :type Gameplay%20Tag%20Container%20Structure%20%28by%20ref%29))
  ;[1 1 1] Add NavInvoker around device when required
  (Playlist|UnbindEventfromOnCurrentPlaylistLoaded (Fortnite|Athena|GameState|GetGameStateAthenafromContext self) (delegate Custom|OnPlaylistLoaded))
  (Class|BPCreativeCommonAILibrary|TryAddNavInvokeronDevice self "/CRD_Creatures/Components/CreaturePlacer_NavInvokerCompClass.CreaturePlacer_NavInvokerCompClass_C") ; Will add Nav Invoker around this device if Nav Mesh on Demand is enabled
)

(fn FadeAmbientLoop
  (params (FadeIn :type Boolean))
  (bind _creatureplacer_ambient_loop (Variables|Default|GetCreaturePlacer_Ambient_Loop))
  (bind _returnvalue (Audio|Components|Audio|IsPlaying _creatureplacer_ambient_loop))
  (bind _creatureplacer_ambient_loop_2 (Variables|Default|GetCreaturePlacer_Ambient_Loop))
  (if FadeIn
    (if (not _returnvalue)
      (Audio|Components|Audio|FadeIn _creatureplacer_ambient_loop_2 (LocalVariables|GetFadeInDuration) 1.0 0.0 "Linear" :!class /Script/Engine.AudioComponent))
    (elif _returnvalue
      (Audio|Components|Audio|FadeOut _creatureplacer_ambient_loop_2 (LocalVariables|GetFadeOutDuration) 0.0 "Linear" :!class /Script/Engine.AudioComponent))))

(fn GetCreatureTypeIndex
  ;[0 1 0.7495 0.25] Creature Type Enum has Cube Random in the first slot, and the arrays that use the index of the enum predated the enum's existance.
  (return (- (Math|Conversions|ToInteger(Byte) (Variables|UserOptions|Basic|GetCreature%20Type)) 1)))

(fn GetCollisionStaticMesh
  (return "0, 0, 0" (Variables|/Script/Engine.StaticMeshComponent|GetStaticMesh (Variables|Debug|GetStaticMeshComponent))))

(fn GetMeshComponents
  (return (Utilities|Array|MakeArray (Variables|Debug|GetCreatureSkeletalMeshComponent) (Variables|Debug|GetStaticMeshComponent))))

(fn IgnoreWorldGridScaleOverride
  (return false))

(fn InitializeDevice
  (CallFunction|SetupConfigurables)
  (if (CallFunction|ShouldUpdateCreatureVisuals)
    (Variables|Internal|SetPrevAITypeIndex (CallFunction|GetCreatureTypeIndex)) ; Set this after checking if the PresetIndex changed
    (CallFunction|UpdateCreatureVisuals)
    (if (Editor|InEditor)
      (CallFunction|OnCreatureVisualsLoadCompleteClient) ; Code delegate does not bind/fire in editor, so manually call the completion callback
    )))

(fn IsPlayingAnimation
  (params (AnimSequence :type Anim%20Sequence%20Object%20Reference))
  (return (and (Components|Animation|IsPlaying (Variables|Debug|GetCreatureSkeletalMeshComponent)) (== (Utilities|Struct|BreakSingleAnimationPlayData (Variables|/Script/Engine.SkeletalMeshComponent|GetAnimationData (Variables|Debug|GetCreatureSkeletalMeshComponent))) (Variables|GetAnimSequence)))))

(fn IsTargetableByLinearMover
  (return true))

(fn MigrateLegacyUserOptions
  (bind (_enableongamephase _enable _disable _aitypepresetindex) (Class|GetClassDefaults "/CRD_Creatures/Device_CreaturePlacer_V2.Device_CreaturePlacer_V2_C"))
  (bind _aitypepresetindex_2 (Variables|Legacy|GetAITypePresetIndex))
  (|CallParentFunction)
  (Utilities|FlowControl|Sequence
    (:then_0
      (if (!= _aitypepresetindex _aitypepresetindex_2)
        (Variables|UserOptions|Basic|SetCreature%20Type (Utilities|Enum|BytetoEnumECreatureType (Math|Conversions|ToByte(Integer) (+ _aitypepresetindex_2 1))))
        (Variables|Legacy|SetAITypePresetIndex _aitypepresetindex)
        (LocalVariables|SetMigrate%20Options true)))
    (:then_1
      (if (LocalVariables|GetMigrate%20Options)
        (OptionData|RefreshPropertyOverrides (Variables|Default|GetToyOptionsComponent))))))

(fn OnPlaylistLoaded
  (params (PlaylistName :type Name) (PlaylistContextTags :type Gameplay%20Tag%20Container%20Structure%20%28by%20ref%29))
  (Class|BPCreativeCommonAILibrary|TryAddNavInvokeronDevice self "/CRD_Creatures/Components/CreaturePlacer_NavInvokerCompClass.CreaturePlacer_NavInvokerCompClass_C") ; Will add Nav Invoker around this device if Nav Mesh on Demand is enabled
)

(fn PhotoMode_GetAllConditionallyHiddenComponents
  (return (Utilities|Map|MakeMap (Variables|Debug|GetCreatureSkeletalMeshComponent) (Utilities|Enum|NotEqual(Enum) (Variables|UserOptions|GetEnableOnGamePhase) "NewEnumerator2"))))

(fn PhotoMode_GetAllHiddenDuringGameComponents
  (return (Utilities|Array|MakeArray (Variables|Default|GetP_CreaturePlacer_Smoke) (Variables|Debug|GetStaticMeshComponent) (Variables|Default|GetCreatureHighlight) (Variables|Default|GetSpotLight) (Variables|Default|GetHoloFloor) (Variables|Default|GetLightRaysB) (Variables|Default|GetLightRaysA))))

(fn SetupConfigurables
  (bind _creature_index (CallFunction|GetCreatureTypeIndex))
  (Variables|Default|SetSpawnAppliedGameplayTag (Utilities|Map|Find (Variables|Default|GetEnemyTypeIntOptionMap) _creature_index))
  (Variables|BuildingProp_CreaturePlacer|SetSelectedSpawnGroup (Utilities|Struct|BreakSpawnGroupVisuals (Utilities|Array|Get(aref) (Variables|BuildingProp_CreaturePlacer|GetSpawnGroupVisuals) _creature_index)))
  (Variables|BuildingProp_CreaturePlacer|SetbHideVFX (not (Variables|UserOptions|GetSpawn%20Effects%20Visibility)))
  ;[1 1 1] Set the Spawn on Game State Logic
  (switch Utilities|FlowControl|Switch|SwitchonEActivationGamePhaseType (Variables|UserOptions|GetEnableOnGamePhase)
    (:NewEnumerator0
      (Variables|BuildingProp_CreaturePlacer|SetGameStateEnable "PreGame")
      (Variables|BuildingProp_CreaturePlacer|SetbEnabledOnGameState false)
      ;[1 1 1] Start/stop SFX based on when it's enabled
      (CallFunction|FadeAmbientLoop false))
    (:NewEnumerator1
      (Variables|BuildingProp_CreaturePlacer|SetGameStateEnable "Warmup")
      (Variables|BuildingProp_CreaturePlacer|SetbEnabledOnGameState true)
      (CallFunction|FadeAmbientLoop true))
    (:NewEnumerator2
      (Variables|BuildingProp_CreaturePlacer|SetGameStateEnable "InProgress")))
  (Variables|BuildingProp_AISpawner|SetDespawnRangeOverride (* (Variables|UserOptions|GetDespawn%20Range) 512.0))
  (Variables|BuildingProp_CreaturePlacer|SetbDestroyPreviousOnSpawn (not (Variables|UserOptions|GetSpawn%20Only%20If%20Needed)))
  (switch Utilities|FlowControl|Switch|SwitchonEDespawnType (Variables|UserOptions|GetDespawn%20Type)
    (:NewEnumerator0
      (Variables|BuildingProp_AISpawner|SetbCanAIDespawnWhenTooFar true)
      (Variables|BuildingProp_AISpawner|SetbUseDistanceToAI true))
    (:NewEnumerator1
      (Variables|BuildingProp_AISpawner|SetbCanAIDespawnWhenTooFar true)
      (Variables|BuildingProp_AISpawner|SetbUseDistanceToAI false))
    (:NewEnumerator2
      (Variables|BuildingProp_AISpawner|SetbCanAIDespawnWhenTooFar false)
      (Variables|BuildingProp_AISpawner|SetbUseDistanceToAI false)))
  (if (>= (Variables|UserOptions|Basic|GetActivation%20Range) 0)
    (Variables|BuildingProp_AISpawner|SetActivationRangeLevel (+ (Variables|UserOptions|Basic|GetActivation%20Range) 1.0))
    (else
      (Variables|BuildingProp_AISpawner|SetActivationRangeLevel 5.0))))

(fn ShouldInvisiblePropsBeAllowed
  (return false))

(fn ShouldUpdateCreatureVisuals
  (return (!= (Variables|Internal|GetPrevAITypeIndex) (CallFunction|GetCreatureTypeIndex))))

(fn SpawnedAtRequestWrapper
  (params (Controller :type Controller%20Object%20Reference) (SpawnedID :type String))
  (GameplayEvent|BroadcastEvent (Variables|UserOptions|Hidden|GetOnSpawnedWithRequestID) 0 (Utilities|Struct|MakeGameplayEventAICreatureSpawnedbyRequestID Controller SpawnedID)))

(fn SpawnWithTransform_Verse
  (params (InTransform :type Transform))
  (bind _returnvalue (CallFunction|GetCreativeRift))
  (if (Utilities|IsValid _returnvalue :!class /Script/Engine.KismetSystemLibrary)
    (Class|BuildingRift|EnableSpawning _returnvalue true)
    (bind _returnvalue_2 (AI|QueueSpawnatTransform (Fortnite|GetAIDirector) _returnvalue (Variables|GetInTransform)))
    (if (Guid|IsValidGuid _returnvalue_2)
      (EventDispatchers|BindEventtoOnSpawnedwithRequestID (Fortnite|GetAIDirector) (EventDispatchers|CreateEvent :!event SpawnedAtRequestWrapper))
      (return (Guid|ToString(Guid) _returnvalue_2))
      (else
        _))
    (else
      _))
  ;[1 1 1] If ID return is invalid then spawn failed to be put in pending list and we should send a failure message back
  (Utilities|String|LogString (Utilities|String|Append (Utilities|GetDisplayName self) " was unable to trigger a spawn at " (Utilities|String|ToString(Transform) (Variables|GetInTransform))))
  (return ""))

(fn ConstructionScript)
