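; Creative creature spawner device that spawns AI creatures from a Deimos rift and broadcasts spawn and elimination events.
; At minigame start it maps user options to parent spawner settings and EQS query, then configures the spawned rift on authority.
; Option changes refresh async-loaded meshes and spawn/destruction area previews.
; It also migrates legacy options, adds a nav invoker for on-demand navmesh, and queues transform spawns from Verse with request IDs.
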(event EventBeginPlay
  (|CallParentFunction)
  (CallFunction|SetSpawnLocation (Variables|Default|GetSpawnPoint))
  (CallFunction|TryRegisterNavInvoker))

(event EventBP_NotifyAIPawnKilled
  (params (KilledAI :type Fort%20AIPawn%20Object%20Reference) (KillingInstigator :type Controller%20Object%20Reference))
  (GameplayEvent|BroadcastEvent (Variables|UserOptions|Events|GetOn%20a%20Creature%20is%20Eliminated) 0 (Utilities|Struct|MakeGameplayEventCreatureEliminated KillingInstigator (Pawn|GetController KilledAI))))

(event Custom|TurnOnEvent
  (params (Instigator :type Controller%20Object%20Reference))
  (CallFunction|OnTurnonReceived Instigator))

(event Custom|TurnOffEvent
  (params (Instigator :type Controller%20Object%20Reference))
  (CallFunction|OnTurnOffReceived Instigator))

(event Custom|DestroySpawnerEvent
  (params (Instigator :type Controller%20Object%20Reference))
  (CallFunction|OnKillRiftReceived Instigator))

(event OnPropertiesLoaded(ToyOptionsComponent)
  (Utilities|FlowControl|Sequence
    (:then_0
      (Utilities|FlowControl|DoOnce false
        (:Completed
          (Components|StaticMesh|SetStaticMesh (Variables|Default|GetBaseMesh) 0)
          (Components|StaticMesh|SetStaticMesh (Variables|Default|GetCubeMesh) 0))))
    (:then_1
      (Networking|CanExecuteCosmeticEvents
        (:True
          (CallFunction|UpdateSpawnLocationVisuals)
          (CallFunction|UpdateSpawnDestructionVisuals)
          (CallFunction|LoadSelectedStaticMeshandMaterials))))))

(event Custom|EliminateCreaturesEvent
  (params (Instigator :type Controller%20Object%20Reference))
  (CallFunction|OnKillAIReceived Instigator))

(event OnAnyPropertyChanged(ToyOptionsComponent)
  (Networking|CanExecuteCosmeticEvents
    (:True
      (CallFunction|UpdateSpawnLocationVisuals)
      (CallFunction|UpdateSpawnDestructionVisuals)
      (CallFunction|LoadSelectedStaticMeshandMaterials))))

(event EventMinigameStarted
  (CallFunction|SetupConfigurables)
  (Networking|SwitchHasAuthority
    (:Authority
      (|CallParentFunction))))

(event EventOnSpawnedRift
  (params (SpawnedRift :type Building%20Rift%20Object%20Reference))
  (bind _deimosrift (Variables|Default|GetDeimosRift))
  (Networking|SwitchHasAuthority
    (:Authority
      (bind _asbp_creative_deimos_rift (Utilities|Casting|CastToBP_CreativeDeimosRift SpawnedRift))
      (Variables|Default|SetDeimosRift _asbp_creative_deimos_rift)
      (if (Utilities|Enum|NotEqual(Enum) (Variables|UserOptions|GetRift%20Spawn%20Destruction%20Type) "None")
        (EventDispatchers|AssignOnBuildingRiftLoadedSettings (Variables|Default|GetDeimosRift) (delegate Custom|RiftLoadedSettings))
        (else
          _))
      (Transformation|AttachActorToActor (Variables|Default|GetDeimosRift) self "None" "KeepWorld" "KeepWorld" "KeepWorld" true) ; Attach to this device. Needed in case something \n(like Verse) moves the device.
      (if (not (Variables|UserOptions|GetSpawner%20Visibility))
        (Variables|/Game/Athena/Deimos/Spawners/RiftSpawners/BP_CreativeDeimosRift.BP_CreativeDeimosRift_C|SetbDamageSelfOnSpawn false _deimosrift)
        (else
          (Variables|/Game/Athena/Deimos/Spawners/RiftSpawners/BP_CreativeDeimosRift.BP_CreativeDeimosRift_C|SetbDamageSelfOnSpawn (Variables|UserOptions|GetDamage%20Spawner%20After%20Spawn) _deimosrift)))
      (Variables|/Game/Athena/Deimos/Spawners/RiftSpawners/BP_CreativeDeimosRift.BP_CreativeDeimosRift_C|SetbHidesVFX (not (Variables|UserOptions|GetSpawn%20Effects%20Visibility)) (Variables|Default|GetDeimosRift))
      (Variables|/Game/Athena/Deimos/Spawners/RiftSpawners/BP_CreativeDeimosRift.BP_CreativeDeimosRift_C|SetbEnableAmbienceSound (Variables|UserOptions|GetEnable%20Ambience%20Sound) (Variables|Default|GetDeimosRift)))))

(event Custom|RiftLoadedSettings
  (bind (_origin _boxextent) (Collision|GetComponentBounds (Variables|/Game/Athena/Deimos/Spawners/RiftSpawners/BP_CreativeDeimosRift.BP_CreativeDeimosRift_C|GetCubeMesh (Variables|Default|GetDeimosRift))))
  (bind (_boxextent_x _boxextent_y _boxextent_z) (Math|Vector|BreakVector _boxextent))
  ; RiftSpawnDestruction
  (BuildingRift|DestroyNearbyBuildingsAdvanced (Variables|Default|GetDeimosRift) (Variables|UserOptions|GetRift%20Spawn%20Destruction%20Type) "PlayAndEditWalls" (/ (CallFunction|GetMinSpawnDistance) 2.0) (/ _boxextent_z 2.0)))

(event MinigameStateChangedDelegate(MinigameProgress)
  (params (Minigame :type Fort%20Minigame%20Object%20Reference) (NewMinigameState :type EFortMinigameState%20Enum :default "PreGame"))
  (CallFunction|UpdateCollision (Utilities|Enum|Equal(Enum) NewMinigameState "PreGame")))

(event Custom|LoadSelectedStaticMeshAndMaterials
  (bind _basemesh (Variables|Default|GetBaseMesh))
  (bind (_base _cube) (CallFunction|GetDesiredMeshMaterial))
  (bind _loaded_object (Creative|Loading|AsyncLoadEditorSafe (Utilities|Struct|BreakMeshMaterial _base))
    (:Loaded
      (bind _asstatic_mesh (Utilities|Casting|CastToStaticMesh _loaded_object))
      (Components|StaticMesh|SetStaticMesh _basemesh _asstatic_mesh)
      (bind _loaded_object_2 (Creative|Loading|AsyncLoadEditorSafe (.Material_9_0C040D9B4C3EF68C191FE79CEE91901E (Utilities|Struct|BreakMeshMaterial _base)))
        (:Loaded
          (bind _asmaterial_interface (Utilities|Casting|CastToMaterialInterface _loaded_object_2))
          (Rendering|Material|SetMaterial _basemesh 0 _asmaterial_interface)
          (bind _loaded_object_3 (Creative|Loading|AsyncLoadEditorSafe (Utilities|Struct|BreakMeshMaterial _cube))
            (:Loaded
              (bind _asstatic_mesh_2 (Utilities|Casting|CastToStaticMesh _loaded_object_3))
              (Components|StaticMesh|SetStaticMesh (Variables|Default|GetCubeMesh) _asstatic_mesh_2)
              (bind _loaded_object_4 (Creative|Loading|AsyncLoadEditorSafe (.Material_9_0C040D9B4C3EF68C191FE79CEE91901E (Utilities|Struct|BreakMeshMaterial _cube)))
                (:Loaded
                  (bind _asmaterial_interface_2 (Utilities|Casting|CastToMaterialInterface _loaded_object_4))
                  (Rendering|Material|SetMaterial (Variables|Default|GetCubeMesh) 0 _asmaterial_interface_2)
                  (Class|CreativeAssetLoadingTrackingComponent|NotifyAssetsLoaded (Variables|Default|GetCreativeAssetLoadingTracking)))))))))))

(event EventBP_NotifyAIPawnSpawned
  (params (SpawnedAI :type Fort%20AIPawn%20Object%20Reference) (SpawnSlotLocation :type Vector%20%28by%20ref%29 :default "0, 0, 0"))
  (bind _returnvalue (Utilities|GetClass SpawnedAI))
  (bind _output (Utilities|ResolveSoftReference _returnvalue))
  (bind _classtotagmap (Variables|Default|GetClassToTagMap))
  (bind (_value _returnvalue_2) (Utilities|Map|Find _classtotagmap _output))
  (bind _aipawnabilitysystemcomponent (Variables|/Script/FortniteGame.FortAIPawn|GetAIPawnAbilitySystemComponent SpawnedAI))
  (Utilities|IsValid SpawnedAI
    (:Is%20Valid
      (if (not (Variables|UserOptions|Basic|GetRestore%20Player%20Shield%20On%20Elimination))
        (Ability|GameplayEffect|AddLooseGameplayTags SpawnedAI (GameplayTags|MakeLiteralGameplayTagContainer "(GameplayTags=((TagName=\"Pawn.Athena.NPC.Deimos.DisableShieldOnElimination\")))") true) ; FORT-684077 | Add Tag to disallow shield rewards
      )
      (if _returnvalue_2
        (bind _returnvalue_3 (Ability|GameplayEffect|AddGrantedTag (GameplayEffects|MakeOutgoingSpec :self _aipawnabilitysystemcomponent :GameplayEffectClass "/Game/Creative/Devices/CreaturePlacer/GameplayEffects/GE_EliminationManagerAddTag.GE_EliminationManagerAddTag_C" :Level 0.0) _value))
        (GameplayEffects|ApplyGameplayEffectSpecToSelf _aipawnabilitysystemcomponent _returnvalue_3)
        (GameplayEvent|BroadcastEvent (Variables|UserOptions|Events|GetOn%20a%20Creature%20is%20Spawned) 0 (Utilities|Struct|MakeGameplayEventCreatureSpawned (Pawn|GetController SpawnedAI)))))))

(event Custom|TryRegisterNavInvoker
  (bind _self self)
  (bind _returnvalue (Fortnite|Athena|GameState|GetGameStateAthenafromContext _self))
  ;[1 1 1] Add NavInvoker around device when required
  (Creative|Networking|Authority/Cosmetic
    (:Authority
      (if (Variables|/Script/FortniteGame.FortGameStateAthena|GetbPlaylistDataIsLoaded _returnvalue)
        (Class|BPCreativeCommonAILibrary|TryAddNavInvokeronDevice self "/CRD_Creatures/Components/CreatureSpawner_NavInvokerCompClass.CreatureSpawner_NavInvokerCompClass_C") ; Will add Nav Invoker around this device if Nav Mesh on Demand is enabled
        (else
          (Playlist|BindEventtoOnCurrentPlaylistLoaded _returnvalue (delegate Custom|OnPlaylistLoaded)))))))

(event Custom|OnPlaylistLoaded
  (params (PlaylistName :type Name) (PlaylistContextTags :type Gameplay%20Tag%20Container%20Structure%20%28by%20ref%29))
  ;[1 1 1] Add NavInvoker around device when required
  (Playlist|UnbindEventfromOnCurrentPlaylistLoaded (Fortnite|Athena|GameState|GetGameStateAthenafromContext self) (delegate Custom|OnPlaylistLoaded))
  (Class|BPCreativeCommonAILibrary|TryAddNavInvokeronDevice self "/CRD_Creatures/Components/CreatureSpawner_NavInvokerCompClass.CreatureSpawner_NavInvokerCompClass_C") ; Will add Nav Invoker around this device if Nav Mesh on Demand is enabled
)

(event Custom|UpdateSpawnDestructionVisuals
  (bind _spawn_area_radius_preview (Variables|Default|GetSpawn%20Area%20Radius%20Preview))
  (bind _relativescale3d (Variables|/Script/Engine.SceneComponent|GetRelativeScale3D _spawn_area_radius_preview))
  (bind _z (.z _relativescale3d))
  (bind _returnvalue (CallFunction|GetMinSpawnDistance))
  (bind _buildinggridsize (Variables|Configurables|GetBuildingGridSize))
  (bind _max_spawn_distance (Variables|UserOptions|Basic|GetMax%20Spawn%20Distance))
  (bind _returnvalue_2 (* _max_spawn_distance _buildinggridsize))
  (bind _buildinggridsize_2 (Variables|Configurables|GetBuildingGridSize))
  (bind _max_spawn_distance_2 (Variables|UserOptions|Basic|GetMax%20Spawn%20Distance))
  (bind _returnvalue_3 (Math|Float|float*float _max_spawn_distance_2 _buildinggridsize_2 0.5))
  (bind _halflargeaicapsuleradius (Variables|Configurables|GetHalfLargeAICapsuleRadius))
  (bind _returnvalue_4 (- _returnvalue_2 _halflargeaicapsuleradius))
  (bind _halflargeaicapsuleradius_2 (Variables|Configurables|GetHalfLargeAICapsuleRadius))
  (bind _returnvalue_5 (/ _returnvalue 2.0))
  (bind _returnvalue_6 (+ _returnvalue_5 _halflargeaicapsuleradius_2))
  (bind _halflargeaicapsuleradius_3 (Variables|Configurables|GetHalfLargeAICapsuleRadius))
  (bind _returnvalue_7 (+ _returnvalue_3 _halflargeaicapsuleradius_3))
  ; DestructionAreaVisualsCode
  (switch Utilities|FlowControl|Switch|SwitchonESpawnDestructionType (Variables|UserOptions|GetRift%20Spawn%20Destruction%20Type)
    (:None
      (ChildActorComponent|SetChildActorClass (Variables|Default|GetDestruction%20Area%20Visualizer) 0))
    (:AroundRift
      (bind _destructionareainterface (Utilities|GetDestructionAreaInterface (Variables|Default|GetDestruction%20Area%20Visualizer) "/Game/Creative/Devices/CaptureArea/BP_AreaVisualizer_ChildActor.BP_AreaVisualizer_ChildActor_C"))
      (Class|AreaVisualizerInterface|SetupDynamicData _destructionareainterface 0.0 _returnvalue_6 _z))
    (:OnSpawnEdge
      (if (>= _returnvalue_4 (/ (Variables|Configurables|GetBuildingGridSize) 2.0))
        (ChildActorComponent|SetChildActorClass (Variables|Default|GetDestruction%20Area%20Visualizer) 0)
        (bind _destructionareainterface_2 (Utilities|GetDestructionAreaInterface (Variables|Default|GetDestruction%20Area%20Edge%20Visualizer) "/Game/Creative/Devices/CaptureArea/BP_AreaVisualizer_Layered_ChildActor.BP_AreaVisualizer_Layered_ChildActor_C")) ; Use the Layered version only when we have an inner radius;\notherwise, it will obstruct the user's vision.\n(It needs a small total radius)
        (Class|AreaVisualizerInterface|SetupDynamicData _destructionareainterface_2 (- _returnvalue_3 _halflargeaicapsuleradius_3) _returnvalue_7 _z)
        (else
          (bind _destructionareainterface_3 (Utilities|GetDestructionAreaInterface (Variables|Default|GetDestruction%20Area%20Visualizer) "/Game/Creative/Devices/CaptureArea/BP_AreaVisualizer_ChildActor.BP_AreaVisualizer_ChildActor_C"))
          (Class|AreaVisualizerInterface|SetupDynamicData _destructionareainterface_3 0.0 _returnvalue_7 _z))))
    (:RiftAndSpawnEdge
      (if (>= (- _returnvalue_4 (+ _halflargeaicapsuleradius (CallFunction|GetMinSpawnDistance))) (Variables|Configurables|GetBuildingGridSize))
        (bind _destructionareainterface_4 (Utilities|GetDestructionAreaInterface (Variables|Default|GetDestruction%20Area%20Visualizer) "/Game/Creative/Devices/CaptureArea/BP_AreaVisualizer_ChildActor.BP_AreaVisualizer_ChildActor_C"))
        (Class|AreaVisualizerInterface|SetupDynamicData _destructionareainterface_4 0.0 _returnvalue_6 _z)
        (else
          _)))
    (:WholeSpawnArea
      _))
  (ChildActorComponent|SetChildActorClass (Variables|Default|GetDestruction%20Area%20Edge%20Visualizer) 0))

(event EditorPreview|EventUpdateEditorPreview
  (params (bHasMoved :type Boolean))
  (ChildActorComponent|SetChildActorClass (Variables|Default|GetDestruction%20Area%20Visualizer) 0))

(fn BP_ShouldTrigger
  (params (TouchingActors :type Array%20of%20Actor%20Object%20References))
  (return true))

(fn ChooseEQSRift
  (LocalVariables|SetbUseVisibility (not (Variables|UserOptions|Basic|GetSpawn%20Through%20Walls)))
  (LocalVariables|SetbUseArea (Utilities|Enum|Equal(Enum) (Variables|UserOptions|GetPreferred%20Spawn%20Location) "Random"))
  (Variables|BuildingProp_DeimosSpawner|SetbSortSlotsByBestScore false)
  (if (LocalVariables|GetbUseVisibility)
    (if (LocalVariables|GetbUseArea)
      (LocalVariables|SetCalculatedIndex 3)
      (else
        (LocalVariables|SetCalculatedIndex 1)))
    (elif (LocalVariables|GetbUseArea)
      (LocalVariables|SetCalculatedIndex 2)
      (else
        (LocalVariables|SetCalculatedIndex 0))))
  (Variables|BuildingProp_DeimosSpawner|SetSelectedEQSQuery (Utilities|Array|Get(acopy) (Variables|CustomizeOptions|GetEQSQueryOptions) (LocalVariables|GetCalculatedIndex))))

(fn GetCreatureTypeAsInt
  (return (Math|Conversions|ToInteger(Byte) (Variables|UserOptions|Basic|GetCreature%20Type))))

(fn GetDesiredMeshMaterial
  (bind _asint (Default|GetSpawnerTypeasInt))
  (bind _returnvalue (* _asint 2))
  (bind _returnvalue_2 (+ _returnvalue 1))
  (bind _meshmaterialarray (Variables|Configurables|GetMeshMaterialArray))
  (if (and (Utilities|Array|IsValidIndex _meshmaterialarray _returnvalue_2) (Utilities|Array|IsValidIndex _meshmaterialarray _returnvalue))
    (return (Utilities|Array|Get(aref) _meshmaterialarray _returnvalue_2) (Utilities|Array|Get(aref) _meshmaterialarray _returnvalue))
    (else
      (return 0 0))))

(fn GetMinSpawnDistance
  (return (* 0.5 (Variables|Configurables|GetBuildingGridSize))))

(fn GetSpawnerTypeAsInt
  (return (Math|Conversions|ToInteger(Byte) (Variables|UserOptions|Basic|GetSpawner%20Type))))

(fn GetSpawnLimit
  (bind _selectedspawnlimit (Variables|BuildingProp_DeimosSpawner|GetSelectedSpawnLimit))
  (bind _spawnlimitoptions (Variables|CustomizeOptions|GetSpawnLimitOptions))
  (bind _returnvalue (Utilities|Array|FindItem _spawnlimitoptions _selectedspawnlimit))
  (if (< _returnvalue 0)
    (return 0)
    (else
      (return (Utilities|Select :Index _returnvalue :Option%200 0 :Option%201 1 :Option%202 2 :Option%203 3 :Option%204 4 :Option%205 5 :Option%206 10 :Option%207 20 :Option%208 30 :Option%209 40 :Option%2010 50 :Option%2011 60 :Option%2012 70 :Option%2013 80 :Option%2014 90 :Option%2015 100)))))

(fn IsTargetableByLinearMover
  (return true))

(fn MigrateLegacyUserOptions
  (bind _despawnrangepresetindex (Class|GetClassDefaults "/CRD_Creatures/Device_CreatureSpawner_V2.Device_CreatureSpawner_V2_C"))
  (bind _activationrangepresetindex (Class|GetClassDefaults "/CRD_Creatures/Device_CreatureSpawner_V2.Device_CreatureSpawner_V2_C"))
  (bind _spawnertypepresetindex (Class|GetClassDefaults "/CRD_Creatures/Device_CreatureSpawner_V2.Device_CreatureSpawner_V2_C"))
  (bind _spawnlimitpresetindex (Class|GetClassDefaults "/CRD_Creatures/Device_CreatureSpawner_V2.Device_CreatureSpawner_V2_C"))
  (bind _spawnradiuspresetindex (Class|GetClassDefaults "/CRD_Creatures/Device_CreatureSpawner_V2.Device_CreatureSpawner_V2_C"))
  (bind _aitypepresetindex (Class|GetClassDefaults "/CRD_Creatures/Device_CreatureSpawner_V2.Device_CreatureSpawner_V2_C"))
  (bind _ainumberpresetindex (Class|GetClassDefaults "/CRD_Creatures/Device_CreatureSpawner_V2.Device_CreatureSpawner_V2_C"))
  (bind _wave_timer (Class|GetClassDefaults "/CRD_Creatures/Device_CreatureSpawner_V2.Device_CreatureSpawner_V2_C"))
  (bind _spawnfreqpresetindex (Class|GetClassDefaults "/CRD_Creatures/Device_CreatureSpawner_V2.Device_CreatureSpawner_V2_C"))
  (bind _spawnertypepresetindex_2 (Variables|Legacy|GetSpawnerTypePresetIndex))
  (bind _spawnfreqpresetindex_2 (Variables|Legacy|GetSpawnFreqPresetIndex))
  (bind _despawnrangepresetindex_2 (Variables|Legacy|GetDespawnRangePresetIndex))
  (bind _activationrangepresetindex_2 (Variables|Legacy|GetActivationRangePresetIndex))
  (bind _spawnlimitpresetindex_2 (Variables|Legacy|GetSpawnLimitPresetIndex))
  (bind _spawnradiuspresetindex_2 (Variables|Legacy|GetSpawnRadiusPresetIndex))
  (bind _aitypepresetindex_2 (Variables|Legacy|GetAITypePresetIndex))
  (bind _ainumberpresetindex_2 (Variables|Legacy|GetAINumberPresetIndex))
  (|CallParentFunction)
  (Utilities|FlowControl|Sequence
    (:then_0
      (Utilities|FlowControl|Sequence
        (:then_0
          ;[0 1 0.6928 0.25] Migrate Despawn Range Preset Index to Despawn Range
          (if (!= _despawnrangepresetindex _despawnrangepresetindex_2)
            (Variables|UserOptions|SetDespawn%20Range (Math|Conversions|ToFloat(Integer) _despawnrangepresetindex_2))
            (Variables|Legacy|SetDespawnRangePresetIndex _despawnrangepresetindex)
            (LocalVariables|SetOptions%20Migrated true)))
        (:then_1
          ;[0 1 0.6928 0.25] Migrate Activation Range Preset Index to Activation Range
          (if (!= _activationrangepresetindex _activationrangepresetindex_2)
            (Variables|UserOptions|Basic|SetActivation%20Range (Math|Conversions|ToFloat(Integer) _activationrangepresetindex_2))
            (Variables|Legacy|SetActivationRangePresetIndex _activationrangepresetindex)
            (LocalVariables|SetOptions%20Migrated true)))
        (:then_2
          ;[0 1 0.6928 0.25] Migrate Spawn Limit Preset Index to Limit Spawned Creatures and Total Spawn limit
          (if (!= _spawnlimitpresetindex _spawnlimitpresetindex_2)
            (Variables|UserOptions|SetLimit%20Spawned%20Creatures true)
            (Variables|UserOptions|SetEnum%20Total%20Spawn%20Limit (Utilities|Enum|BytetoEnumESpawnLimit (Math|Conversions|ToByte(Integer) (- _spawnlimitpresetindex_2 1))))
            (Variables|Legacy|SetSpawnLimitPresetIndex _spawnlimitpresetindex)
            (LocalVariables|SetOptions%20Migrated true)))
        (:then_3
          ;[0 1 0.6928 0.25] Migrate Spawn Radius Preset Index to Spawn Radius
          (if (!= _spawnradiuspresetindex _spawnradiuspresetindex_2)
            (Variables|UserOptions|Basic|SetMax%20Spawn%20Distance (/ _spawnradiuspresetindex_2 512.0))
            (Variables|Legacy|SetSpawnRadiusPresetIndex _spawnradiuspresetindex)
            (LocalVariables|SetOptions%20Migrated true)))
        (:then_4
          ;[0 1 0.5056 0.25] Migrate AITypePresetIndex to Creature Type Enum
          (if (!= _aitypepresetindex _aitypepresetindex_2)
            (Variables|UserOptions|Basic|SetCreature%20Type (Utilities|Enum|BytetoEnumECreatureType (Math|Conversions|ToByte(Integer) _aitypepresetindex_2)))
            (Variables|Legacy|SetAITypePresetIndex _aitypepresetindex)
            (LocalVariables|SetOptions%20Migrated true)))
        (:then_5
          ;[0 1 0.5367 0.25] Migrate AINumberPresetIndex to NumberOfCreatures
          (if (!= _ainumberpresetindex _ainumberpresetindex_2)
            (Variables|UserOptions|Basic|SetNumber%20of%20Creatures (+ _ainumberpresetindex_2 1))
            (Variables|Legacy|SetAINumberPresetIndex _ainumberpresetindex)
            (LocalVariables|SetOptions%20Migrated true)))
        (:then_6
          ;[0 1 0.7984 0.25] Migrate Spawn Freq to Wave Timer
          (if (!= _spawnfreqpresetindex _spawnfreqpresetindex_2)
            (Variables|Legacy|SetWave%20Timer (+ _spawnfreqpresetindex_2 1))
            (Variables|Legacy|SetSpawnFreqPresetIndex _spawnfreqpresetindex)
            (LocalVariables|SetOptions%20Migrated true)))
        (:then_7
          ;[0 1 0.7984 0.25] Migrate Wave Timer to Wave Time Spawner (This option is now a regular int. Wave Timer was used to get a FortDifficultyOptionEncounter_SpawnTiming class, which was passed to the encounter info via the spawned rift so that it could supply a float. This is much cleaner and better for UEFN.)
          (if (!= _wave_timer (Variables|Legacy|GetWave%20Timer))
            (Variables|UserOptions|SetWave%20Spawn%20Timer (Utilities|Map|Find (Variables|Default|GetWaveTimerMigrateMap) (Variables|Legacy|GetWave%20Timer))) ; Set Wave Spawn Timer based on which FortDifficultyOptionEncounter_SpawnTiming class this Wave Timer value was associated with
            (Variables|Legacy|SetWave%20Timer _wave_timer)
            (LocalVariables|SetOptions%20Migrated true)))
        (:then_8
          (if (!= _spawnertypepresetindex _spawnertypepresetindex_2)
            (Variables|UserOptions|Basic|SetSpawner%20Type (Utilities|Enum|BytetoEnumECreatureSpawnerType (Math|Conversions|ToByte(Integer) (+ _spawnertypepresetindex_2 1))))
            (Variables|Legacy|SetSpawnerTypePresetIndex _spawnertypepresetindex)
            (LocalVariables|SetOptions%20Migrated true)))))
    (:then_1
      (if (LocalVariables|GetOptions%20Migrated)
        (OptionData|RefreshPropertyOverrides (Variables|Default|GetToyOptionsComponent))))))

(fn OnPlaylistLoaded
  (params (PlaylistName :type Name) (PlaylistContextTags :type Gameplay%20Tag%20Container%20Structure%20%28by%20ref%29))
  (Class|BPCreativeCommonAILibrary|TryAddNavInvokeronDevice self "/CRD_Creatures/Components/CreatureSpawner_NavInvokerCompClass.CreatureSpawner_NavInvokerCompClass_C") ; Will add Nav Invoker around this device if Nav Mesh on Demand is enabled
)

(fn SetupConfigurables
  (bind _limit_spawned_creatures (Variables|UserOptions|GetLimit%20Spawned%20Creatures))
  ;[0.7438 1 0.2276 0.5] Wave and Creatures settings
  (Variables|BuildingProp_DeimosSpawner|SetSelectedPawnNumberCap (Utilities|Array|Get(acopy) (Variables|CustomizeOptions|GetPawnNumberCapOptions) (- (Variables|UserOptions|Basic|GetNumber%20of%20Creatures) 1)))
  (Variables|BuildingProp_DeimosSpawner|SetSelectedSpawnGroup (Utilities|Array|Get(acopy) (Variables|CustomizeOptions|GetSpawnGroupOptions) (Default|GetCreatureTypeasInt)))
  (Variables|BuildingProp_DeimosSpawner|SetSpawnTimingOverrideValue (Math|Conversions|ToFloat(Integer) (Variables|UserOptions|GetWave%20Spawn%20Timer)))
  (Variables|BuildingProp_DeimosSpawner|SetbSpawnTimingOverrideEnabled true) ; This must be true, or else the device's timer value won't be applied
  (Variables|BuildingProp_DeimosSpawner|SetSelectedSpawnLimit (Utilities|Array|Get(acopy) (Variables|CustomizeOptions|GetSpawnLimitOptions) (select _limit_spawned_creatures (+ (Math|Conversions|ToInteger(Byte) (Variables|UserOptions|GetEnum%20Total%20Spawn%20Limit)) 1) 0)))
  (Variables|BuildingProp_DeimosSpawner|SetbNoSpawnLimitEnabled (not _limit_spawned_creatures))
  ;[0.7872 1 0.4547 0.5] Spawn and despawn settings
  (if (>= (Variables|UserOptions|Basic|GetActivation%20Range) 0)
    (Variables|BuildingProp_AISpawner|SetActivationRangeLevel (+ (Variables|UserOptions|Basic|GetActivation%20Range) 1.0))
    (else
      (Variables|BuildingProp_AISpawner|SetActivationRangeLevel 5.0)))
  (switch Utilities|FlowControl|Switch|SwitchonEDespawnType (Variables|UserOptions|GetDespawn%20Type)
    (:NewEnumerator0
      (Variables|BuildingProp_AISpawner|SetbCanAIDespawnWhenTooFar true)
      (Variables|BuildingProp_AISpawner|SetbUseDistanceToAI true))
    (:NewEnumerator1
      (Variables|BuildingProp_AISpawner|SetbCanAIDespawnWhenTooFar true)
      (Variables|BuildingProp_AISpawner|SetbUseDistanceToAI false))
    (:NewEnumerator2
      (Variables|BuildingProp_AISpawner|SetbCanAIDespawnWhenTooFar false)
      (Variables|BuildingProp_AISpawner|SetbUseDistanceToAI false)))
  (Variables|BuildingProp_AISpawner|SetDespawnRangeOverride (* (Variables|UserOptions|GetDespawn%20Range) 512.0))
  (if (Variables|UserOptions|Basic|GetSpawn%20Through%20Walls)
    (Variables|BuildingProp_DeimosSpawner|SetbDamageAllBuildingsOnSpawn (Variables|UserOptions|GetDamage%20All%20Structures%20at%20Spawn%20Location))
    (Variables|BuildingProp_DeimosSpawner|SetbDamageBuildingsOnSpawn (Variables|UserOptions|GetDamage%20Structures%20at%20Spawn%20Location))
    (else
      ;;[1 1 1] If we can't spawn through walls, we can't spawn inside them to destroy them
      (Variables|BuildingProp_DeimosSpawner|SetbDamageAllBuildingsOnSpawn false)
      (Variables|BuildingProp_DeimosSpawner|SetbDamageBuildingsOnSpawn false)))
  (Variables|BuildingProp_DeimosSpawner|SetOverrideQueryRadius (* (Variables|UserOptions|Basic|GetMax%20Spawn%20Distance) 512.0))
  (Variables|BuildingProp_DeimosSpawner|SetSpawnLocationType (Variables|UserOptions|GetPreferred%20Spawn%20Location))
  (CallFunction|ChooseEQSRift)
  ;[1 0.9965 0.5208 0.5] Spawner settings
  (Variables|BuildingProp_DeimosSpawner|SetbInvulnerable (Variables|UserOptions|GetInvincible%20Spawner))
  (Variables|BuildingProp_DeimosSpawner|SetbVisible (Variables|UserOptions|GetSpawner%20Visibility))
  (Variables|BuildingProp_DeimosSpawner|SetbEnabledOnMinigameStart (Variables|UserOptions|GetEnabled%20at%20Game%20Start))
  (Variables|BuildingProp_AISpawner|SetCreativeRiftClass (Utilities|Array|Get(acopy) (Variables|CustomizeOptions|GetDeimosSpawnerOptions) (Default|GetSpawnerTypeasInt))))

(fn ShouldInvisiblePropsBeAllowed
  (return false))

(fn SpawnedAtRequestWrapper
  (params (Controller :type Controller%20Object%20Reference) (SpawnedID :type String))
  (GameplayEvent|BroadcastEvent (Variables|UserOptions|Hidden|GetOnSpawnedWithRequestID) 0 (Utilities|Struct|MakeGameplayEventAICreatureSpawnedbyRequestID Controller SpawnedID)))

(fn SpawnWithTransform_Verse
  (params (InTransform :type Transform))
  (bind _returnvalue (AI|QueueSpawnatTransform (Fortnite|GetAIDirector) (Variables|Default|GetDeimosRift) (Variables|GetInTransform)))
  (if (Guid|IsValidGuid _returnvalue)
    (EventDispatchers|BindEventtoOnSpawnedwithRequestID (Fortnite|GetAIDirector) (EventDispatchers|CreateEvent :!event SpawnedAtRequestWrapper))
    (return (Guid|ToString(Guid) _returnvalue))
    (else
      ;[1 1 1] If ID return is invalid then spawn failed to be put in pending list and we should send a failure message back
      (Utilities|String|LogString (Utilities|String|Append (Utilities|GetDisplayName self) " was unable to trigger a spawn at " (Utilities|String|ToString(Transform) (Variables|GetInTransform))))
      (return ""))))

(fn UpdateCollision
  (params (Collision%20On :type Boolean))
  (bind _returnvalue (not Collision%20On))
  (CallFunction|AdjustCollision _returnvalue (Variables|Default|GetBaseMesh))
  (CallFunction|AdjustCollision _returnvalue (Variables|Default|GetCubeMesh)))

(fn UpdateSpawnLocationVisuals
  (Transformation|SetRelativeScale3D (Variables|Default|GetSpawn%20Area%20Radius%20Preview) (Math|Vector|MakeVector (* (Variables|UserOptions|Basic|GetMax%20Spawn%20Distance) (Variables|Configurables|GetBuildingGridSize)) (* (Variables|UserOptions|Basic|GetMax%20Spawn%20Distance) (Variables|Configurables|GetBuildingGridSize)) (.z (Variables|/Script/Engine.SceneComponent|GetRelativeScale3D (Variables|Default|GetSpawn%20Area%20Radius%20Preview))))))

(fn ConstructionScript
  (CallFunction|LoadSelectedStaticMeshandMaterials))
